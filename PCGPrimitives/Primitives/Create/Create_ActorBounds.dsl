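; PCG primitive graph Create_ActorBounds, which by its name creates data from actor bounds.
; A core-process subgraph feeds a prepare-outputs subgraph, whose result becomes the graph output.
; Both subgraphs are overridable through graph parameters, with the ActorBounds core and template output graphs as defaults.
; Input-processing and fallback parameters are also declared, but this graph has no input and leaves them unused.
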(graph Create_ActorBounds
  (params (InputsProcessingGraph :type (SoftObject /Script/PCG.PCGGraphInterface) :default /PCGPrimitives/Primitives/_Template_/Primitives_InputsProcessing.Primitives_InputsProcessing) (InputsFallbackGraph :type (SoftObject /Script/PCG.PCGGraphInterface) :default /PCGPrimitives/Primitives/Shared/InputsFallback/InputsFallback_Empty.InputsFallback_Empty) (CoreProcessGraph :type (SoftObject /Script/PCG.PCGGraphInterface) :default /PCGPrimitives/Primitives/Create/Subgraphs/Create_ActorBounds_CoreProcess.Create_ActorBounds_CoreProcess) (PrepareOutputsGraph :type (SoftObject /Script/PCG.PCGGraphInterface) :default /PCGPrimitives/Primitives/_Template_/Primitives_PrepareOutputs.Primitives_PrepareOutputs))
  ;[0 0 0 0.5 32] Core Process
  (bind Subgraph_3 (Subgraph :Graph /PCGPrimitives/Primitives/Create/Subgraphs/Create_ActorBounds_CoreProcess.Create_ActorBounds_CoreProcess :SubgraphOverride-param CoreProcessGraph))
  ;[0 0 0 0.5 32] Prepare Output
  (bind Subgraph_4 (Subgraph :Graph /PCGPrimitives/Primitives/_Template_/Primitives_PrepareOutputs.Primitives_PrepareOutputs :SubgraphOverride-param PrepareOutputsGraph))
  (connect Subgraph_3 Out Subgraph_4 In)
  (output Out Subgraph_4))
