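; PCG primitive graph that creates extruded meshes, by default from closed splines and 2D polygons.
; Tags its input "Input:0", runs it through shared input management with swappable fallback and processing graphs, then feeds an overridable core extrusion subgraph.
; That subgraph receives height range, height attribute, material, dynamic-mesh toggle and output mesh name and path.
; The graph wires no output, and its PrepareOutputsGraph parameter is unused.
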(graph Create_Mesh_Extrude
  (params (InputsProcessingGraph :type (SoftObject /Script/PCG.PCGGraphInterface) :default /PCGPrimitives/Primitives/Shared/InputsProcessing/InputsProcessing_ConvertPoly2DsAndForwardClosedSplines.InputsProcessing_ConvertPoly2DsAndForwardClosedSplines) (InputsFallbackGraph :type (SoftObject /Script/PCG.PCGGraphInterface) :default /PCGPrimitives/Primitives/_Template_/Primitives_InputsFallback.Primitives_InputsFallback) (CoreProcessGraph :type (SoftObject /Script/PCG.PCGGraphInterface) :default /PCGPrimitives/Primitives/Create/Subgraphs/Create_Mesh_Extrude_CoreProcess.Create_Mesh_Extrude_CoreProcess) (PrepareOutputsGraph :type (SoftObject /Script/PCG.PCGGraphInterface)) (DynamicMesh :type Bool :default true) (Material :type (SoftObject /Script/Engine.MaterialInterface)) (MinHeight :type Double :default 1.5e+03) (MaxHeight :type Double :default 3e+03) (OutMeshName :type String :default "ExtrudedMesh") (OutMeshPath :type String :default "/Game/PCG/_GENERATED") (HeightAttribute :type String))
  (inputs (PrimaryInput :type Any))
  (bind AddTags_0 (Add%20Tags :TagsToAdd "Input:0"))
  ;[0 0 0 0.5 32] Inputs
  (bind Subgraph_0 (Subgraph :Graph /PCGPrimitives/Primitives/_Template_/Subgraphs/Primitives_InputsManagement.Primitives_InputsManagement :InputsFallbackGraph InputsFallbackGraph :InputsProcessingGraph InputsProcessingGraph))
  ;[0 0 0 0.5 32] Core Process
  (bind Subgraph_3 (Subgraph :Graph /PCGPrimitives/Primitives/Create/Subgraphs/Create_Mesh_Extrude_CoreProcess.Create_Mesh_Extrude_CoreProcess :DynamicMesh DynamicMesh :HeightAttribute HeightAttribute :Material Material :MaxHeight MaxHeight :MinHeight MinHeight :OutMeshName OutMeshName :OutMeshPath OutMeshPath :SubgraphOverride-param CoreProcessGraph))
  (connect DefaultInputNode PrimaryInput AddTags_0 In)
  (connect AddTags_0 Out Subgraph_0 In)
  (connect Subgraph_0 Out Subgraph_3 In))
