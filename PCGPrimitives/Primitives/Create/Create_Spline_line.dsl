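; PCG primitive graph that creates a spline line, parameterized by length, position, rotation and an at-world-origin toggle.
; The primary input is tagged, passed through templated input management (by default converting to points, with an empty fallback), then fed to the line core-process subgraph and an output-preparation subgraph.
; Input processing, fallback, core process and output preparation are exposed as graph parameters, so each stage is swappable.
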(graph Create_Spline_line
  (params (InputsProcessingGraph :type (SoftObject /Script/PCG.PCGGraphInterface) :default /PCGPrimitives/Primitives/Shared/InputsProcessing/InputsProcessing_ConvertToPoints.InputsProcessing_ConvertToPoints) (InputsFallbackGraph :type (SoftObject /Script/PCG.PCGGraphInterface) :default /PCGPrimitives/Primitives/Shared/InputsFallback/InputsFallback_Empty.InputsFallback_Empty) (CoreProcessGraph :type (SoftObject /Script/PCG.PCGGraphInterface) :default /PCGPrimitives/Primitives/Create/Subgraphs/Create_Spline_line_CoreProcess.Create_Spline_line_CoreProcess) (PrepareOutputsGraph :type (SoftObject /Script/PCG.PCGGraphInterface) :default /PCGPrimitives/Primitives/_Template_/Primitives_PrepareOutputs.Primitives_PrepareOutputs) (Length :type Double :default 1e+03) (Position :type Vector) (Rotation :type Rotator) (AtWorldOrigin :type Bool :default true))
  (inputs (PrimaryInput :type Any))
  (bind AddTags_0 (Add%20Tags :TagsToAdd "Input:0"))
  ;[0 0 0 0.5 32] Inputs
  (bind Subgraph_0 (Subgraph :Graph /PCGPrimitives/Primitives/_Template_/Subgraphs/Primitives_InputsManagement.Primitives_InputsManagement :InputsFallbackGraph InputsFallbackGraph :InputsProcessingGraph InputsProcessingGraph))
  ;[0 0 0 0.5 32] Core Process
  (bind Subgraph_3 (Subgraph :Graph /PCGPrimitives/Primitives/Create/Subgraphs/Create_Spline_line_CoreProcess.Create_Spline_line_CoreProcess :AtWorldOrigin AtWorldOrigin :Length Length :Position Position :Rotation Rotation :SubgraphOverride-param CoreProcessGraph))
  ;[0 0 0 0.5 32] Prepare Output
  (bind Subgraph_2 (Subgraph :Graph /PCGPrimitives/Primitives/_Template_/Primitives_PrepareOutputs.Primitives_PrepareOutputs :SubgraphOverride-param PrepareOutputsGraph))
  (connect DefaultInputNode PrimaryInput AddTags_0 In)
  (connect AddTags_0 Out Subgraph_0 In)
  (connect Subgraph_3 Out Subgraph_2 In)
  (connect Subgraph_0 Out Subgraph_3 In)
  (output Out Subgraph_2))
